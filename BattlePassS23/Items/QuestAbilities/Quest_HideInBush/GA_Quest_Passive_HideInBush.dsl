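; Passive quest ability that credits the player for hiding in bushes they threw, counting each bush once.
; An async wait for ThrowBush gameplay events records each thrown bush from the payload's OptionalObject.
; When a configured tag is added, it scans actors overlapping the pawn for thrown, unused Apollo_BigBush actors, marks them used and sends a quest stat event.
; The same logic is duplicated in ActivateAbility and FailedToActivatePassiveAbility.
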(event Ability|EventFailedToActivatePassiveAbility
  (params (ActorInfo :type Gameplay%20Ability%20Actor%20Info%20Structure))
  (bind (_asynctaskproxy _payload) (Ability|Async|WaitGameplayEventToActor (Variables|GAVars|Data|GetOwningPlayerPawn) "(TagName=\"athena.event.gameplay.ThrowBush\")")
    (:then
      (Ability|Tasks|WaitGameplayTagAdd (Variables|Default|GetTag)
        (:Added
          (for _array_element (Collision|GetOverlappingActors (Variables|GAVars|Data|GetOwningPlayerPawn) (Variables|Default|GetClass%20Filter) :!class /Script/Engine.Actor)
            (bind _asapollo_big_bush (Utilities|Casting|CastToApollo_BigBush _array_element))
            (if (Utilities|Array|ContainsItem (Variables|Default|GetUsed%20Bushes) _asapollo_big_bush)
              _
              (elif (Utilities|Array|ContainsItem (Variables|Default|GetThrown%20Bushes) _asapollo_big_bush)
                (Utilities|Array|AddUnique (Variables|Default|GetUsed%20Bushes) _asapollo_big_bush)
                (Quest|SendComplexCustomStatEvent (Variables|GAVars|Data|GetOwningPlayerQuestManager) 0 (Variables|Default|GetAdditional%20Source%20Tags) "(GameplayTags=)" 1)))))))
    (:EventReceived
      (Utilities|Array|AddUnique (Variables|Default|GetThrown%20Bushes) (.OptionalObject (Utilities|Struct|BreakGameplayEventData _payload))))))

(event Ability|EventActivateAbility
  (|CallParentFunction)
  (bind (_asynctaskproxy _payload) (Ability|Async|WaitGameplayEventToActor (Variables|GAVars|Data|GetOwningPlayerPawn) "(TagName=\"athena.event.gameplay.ThrowBush\")")
    (:then
      (Ability|Tasks|WaitGameplayTagAdd (Variables|Default|GetTag)
        (:Added
          (for _array_element (Collision|GetOverlappingActors (Variables|GAVars|Data|GetOwningPlayerPawn) (Variables|Default|GetClass%20Filter) :!class /Script/Engine.Actor)
            (bind _asapollo_big_bush (Utilities|Casting|CastToApollo_BigBush _array_element))
            (if (Utilities|Array|ContainsItem (Variables|Default|GetUsed%20Bushes) _asapollo_big_bush)
              _
              (elif (Utilities|Array|ContainsItem (Variables|Default|GetThrown%20Bushes) _asapollo_big_bush)
                (Utilities|Array|AddUnique (Variables|Default|GetUsed%20Bushes) _asapollo_big_bush)
                (Quest|SendComplexCustomStatEvent (Variables|GAVars|Data|GetOwningPlayerQuestManager) 0 (Variables|Default|GetAdditional%20Source%20Tags) "(GameplayTags=)" 1)))))))
    (:EventReceived
      (Utilities|Array|AddUnique (Variables|Default|GetThrown%20Bushes) (.OptionalObject (Utilities|Struct|BreakGameplayEventData _payload))))))
